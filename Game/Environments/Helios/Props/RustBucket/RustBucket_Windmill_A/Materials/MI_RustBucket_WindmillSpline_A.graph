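# Material for the RustBucket windmill spline prop, set up as two-sided Default Lit with Nanite usage and world-space normals.
# Its graph holds a SplineThicken function call, with one placeholder-named scalar parameter driving both base and tip width for a uniform thickness.
# No connection from the function to a material output appears in the graph, so it looks unfinished.
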Material /Game/Environments/Helios/Props/RustBucket/RustBucket_Windmill_A/Materials/MI_RustBucket_WindmillSpline_A
graph {
 properties {bTangentSpaceNormal=false, bUsedWithNanite=true, pixelDepthOffsetMode=PDOM_Legacy, twoSided=true, usedShadingModels=Default Lit}
 nodes [
  n1 = ScalarParameter() {defaultValue=7, parameterName=asdfasf}
  MaterialFunctionCall(WidthBase=@n1, WidthTip=@n1) {materialFunction=&SplineThicken}
 ]
}
